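; Frontend item view settings manager that decides how Fortnite menus present a previewed item, based on camera, game mode and item type.
; GetItemViewSettings returns defaults for shuffle tiles, otherwise sequentially sets rotation offset and mode, zoom, hero placement pad and fixed camera.
; Pickaxes and gliders with a non-zero framing center offset override rotation to BoundsPivot.
; Helpers detect hero-previewed items and Juno building sets or props by primary asset type.
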(fn GetItemViewSettings
  (params (Camera :type EFrontEndCamera%20Enum :default "Invalid") (GameMode :type ESubGame%20Enum :default "Campaign") (ItemDefinition :type Item%20Definition%20Base%20Object%20Reference) (SubSlotIndex :type Integer))
  (bind _returnvalue (Class|FortItemFunctionLibrary|GetItemType ItemDefinition))
  (bind _returnvalue_2 (FortItemType|GetTemplateType _returnvalue))
  (bind _returnvalue_3 (Class|FortItemFunctionLibrary|GetItemType ItemDefinition))
  (bind _returnvalue_4 (FortItemType|GetTemplateType _returnvalue_3))
  (bind _returnvalue_5 0.0)
  (bind (_asathena_sky_dive_contrail_item_definition _bsuccess) (Utilities|Casting|CastToAthenaSkyDiveContrailItemDefinition (Variables|GetItemDefinition)))
  (bind (_asathena_toy_item_definition _bsuccess_2) (Utilities|Casting|CastToAthenaToyItemDefinition ItemDefinition))
  (bind (_asathena_spray_item_definition _bsuccess_3) (Utilities|Casting|CastToAthenaSprayItemDefinition ItemDefinition))
  (bind (_asathena_pet_carrier_item_definition _bsuccess_4) (Utilities|Casting|CastToAthenaPetCarrierItemDefinition ItemDefinition))
  (bind _floatrangebound (Utilities|Struct|MakeFloatRangeBound "Inclusive" _returnvalue_5))
  (bind _floatrangebound_2 (Utilities|Struct|MakeFloatRangeBound "Inclusive" 1.0))
  (bind _floatrange (Utilities|Struct|MakeFloatRange _floatrangebound _floatrangebound_2))
  (Utilities|FlowControl|Sequence
    (:then_0
      ;[1 1 0] Early Out for Random Customization Items
      (if (Class|FortItemFunctionLibrary|IsShuffleTile (Variables|GetItemDefinition))
        (return (LocalVariables|GetReturn%20Value))))
    (:then_1
      (Utilities|FlowControl|Sequence
        (:then_0
          ;[1 1 1] Camera Rotation Offsets
          (if (CallFunction|IsHeroOrPreviewsOnHero (Variables|GetItemDefinition) SubSlotIndex)
            (if (and (Utilities|Array|ContainsItem (Utilities|Array|MakeArray "Home" "Store" "StoreItemInspect") Camera) (Utilities|Enum|NotEqual(Enum) GameMode "Campaign"))
              (Utilities|Struct|SetmembersinFortItemViewSettings (LocalVariables|GetReturn%20Value) "15.000000, 0,-2.000000")
              (elif (Utilities|Array|ContainsItem (Utilities|Array|MakeArray "Cosmetics") Camera)
                (Utilities|Struct|SetmembersinFortItemViewSettings (LocalVariables|GetReturn%20Value) "15.000000, 0,-2.000000")
                (elif (Utilities|Array|ContainsItem (Utilities|Array|MakeArray "SmallCosmetics") Camera)
                  (Utilities|Struct|SetmembersinFortItemViewSettings (LocalVariables|GetReturn%20Value) "5.000000, 0,-2.000000"))))))
        (:then_1
          ;[1 1 1] Rotation Mode
          (if _bsuccess_3
            (Utilities|Struct|SetmembersinFortItemViewSettings (LocalVariables|GetReturn%20Value) "None")
            (elif (or (CallFunction|IsJunoBuilding (Variables|GetItemDefinition)) (Utilities|Array|ContainsItem (Utilities|Array|MakeArray "Schematic" "Hero" "Weapon" "Trap" "SparksGuitar" "SparksBass" "SparksKeyboard" "SparksMicrophone" "SparksDrums" "CosmeticShoes" "CosmeticMimosa") _returnvalue_4))
              (Utilities|Struct|SetmembersinFortItemViewSettings (LocalVariables|GetReturn%20Value) "BoundsPivot")
              (elif (or (Utilities|Array|ContainsItem (Utilities|Array|MakeArray "AthenaGlider" "AthenaPickaxe" "AthenaItemWrap") _returnvalue_4) _bsuccess_2)
                (Utilities|Struct|SetmembersinFortItemViewSettings (LocalVariables|GetReturn%20Value) "World")
                (elif (Utilities|Array|ContainsItem (Utilities|Array|MakeArray "AthenaDance" "AthenaBackpack" "AthenaSkyDiveContrail" "AthenaCharacter" "AthenaVehicleCosmetic" "SparksAura") _returnvalue_4)
                  (Utilities|Struct|SetmembersinFortItemViewSettings (LocalVariables|GetReturn%20Value) "Relative"))))))
        (:then_2
          ;[1 1 1] Supports Zooming
          (if true
            (if (or (Utilities|Array|ContainsItem (Utilities|Array|MakeArray "Schematic" "Hero" "Weapon" "Trap" "AthenaBackpack" "AthenaCharacter" "AthenaItemWrap" "AthenaVehicleCosmetic" "SparksGuitar" "SparksBass" "SparksKeyboard" "SparksMicrophone" "SparksDrums" "SparksAura" "CosmeticMimosa") (FortItemType|GetTemplateType (Class|FortItemFunctionLibrary|GetItemType ItemDefinition))) (CallFunction|IsJunoBuilding (Variables|GetItemDefinition)))
              (Utilities|Struct|SetmembersinFortItemViewSettings (LocalVariables|GetReturn%20Value) true))
            (else
              ;;[1 0.06918 0.1489] Zooming Off for now.
              (Utilities|Struct|SetmembersinFortItemViewSettings (LocalVariables|GetReturn%20Value) false))))
        (:then_3
          ;[1 1 1] Zoom Levels
          (if (Utilities|Array|ContainsItem (Utilities|Array|MakeArray "Hero" "AthenaCharacter" "SparksAura") _returnvalue_2)
            (Utilities|Struct|SetmembersinFortItemViewSettings (LocalVariables|GetReturn%20Value) (select (Utilities|Enum|Equal(Enum) Camera "Home") 0.8 0.0) (Utilities|Struct|MakeFloatRange (Utilities|Struct|MakeFloatRangeBound "Inclusive") (Utilities|Struct|MakeFloatRangeBound "Inclusive" (Math|Float|SelectFloat 1.0 0.8 (Class|FortUICameraFramingFunctionLibrary|IsNewonPawnZoomLogicEnabled)))))
            (elif (Utilities|Array|ContainsItem (Utilities|Array|MakeArray "AthenaGlider") _returnvalue_2)
              (Utilities|Struct|SetmembersinFortItemViewSettings (LocalVariables|GetReturn%20Value) 0.0 (Utilities|Struct|MakeFloatRange (Utilities|Struct|MakeFloatRangeBound "Inclusive" 0.0) (Utilities|Struct|MakeFloatRangeBound "Inclusive" 1.0)))
              (elif (Utilities|Array|ContainsItem (Utilities|Array|MakeArray "AthenaPickaxe" "AthenaItemWrap") _returnvalue_2)
                (Utilities|Struct|SetmembersinFortItemViewSettings (LocalVariables|GetReturn%20Value) _returnvalue_5 _floatrange)
                (elif (Utilities|Array|ContainsItem (Utilities|Array|MakeArray "AthenaBackpack") _returnvalue_2)
                  (Utilities|Struct|SetmembersinFortItemViewSettings (LocalVariables|GetReturn%20Value) (Math|Float|SelectFloat 1.0 0.5 _bsuccess_4) _floatrange)
                  (elif (CallFunction|IsJunoBuilding (Variables|GetItemDefinition))
                    (Utilities|Struct|SetmembersinFortItemViewSettings (LocalVariables|GetReturn%20Value) 0.0 (Utilities|Struct|MakeFloatRange (Utilities|Struct|MakeFloatRangeBound "Inclusive" 0.0) (Utilities|Struct|MakeFloatRangeBound "Inclusive" 0.7)))))))))
        (:then_4
          ;[1 1 1] Placement Actor Pad under all things with Heroes *except* Skydive Contrails
          (Utilities|Struct|SetmembersinFortItemViewSettings (LocalVariables|GetReturn%20Value) (and (CallFunction|IsHeroOrPreviewsOnHero (Variables|GetItemDefinition) SubSlotIndex) (not _bsuccess))))
        (:then_5
          ;[1 1 1] Fixed Cameras
          (if (Utilities|Array|ContainsItem (Utilities|Array|MakeArray "AthenaSkyDiveContrail" "AthenaGlider") (FortItemType|GetTemplateType (Class|FortItemFunctionLibrary|GetItemType ItemDefinition)))
            (Utilities|Struct|SetmembersinFortItemViewSettings (LocalVariables|GetReturn%20Value) true)))
        (:then_6
          ;[1 1 1] Support offset bounds pivot center for pickaxes that need them (mostly Dual Handed ones)
          (bind _asathena_pickaxe_item_definition (Utilities|Casting|CastToAthenaPickaxeItemDefinition ItemDefinition))
          (if (not (Math|Vector|VectorIsZero (Variables|/Script/FortniteGame.AthenaPickaxeItemDefinition|GetCameraFramingBoundsCenterOffset _asathena_pickaxe_item_definition)))
            (Utilities|Struct|SetmembersinFortItemViewSettings (LocalVariables|GetReturn%20Value) "BoundsPivot" (Variables|/Script/FortniteGame.AthenaPickaxeItemDefinition|GetCameraFramingBoundsCenterOffset _asathena_pickaxe_item_definition))))
        (:then_7
          ;[1 1 1] Support offset bounds pivot center for gliders that need them
          (bind _asathena_glider_item_definition (Utilities|Casting|CastToAthenaGliderItemDefinition ItemDefinition))
          (if (not (Math|Vector|VectorIsZero (Variables|/Script/FortniteGame.AthenaGliderItemDefinition|GetCameraFramingBoundsCenterOffset _asathena_glider_item_definition)))
            (Utilities|Struct|SetmembersinFortItemViewSettings (LocalVariables|GetReturn%20Value) "BoundsPivot" (Variables|/Script/FortniteGame.AthenaGliderItemDefinition|GetCameraFramingBoundsCenterOffset _asathena_glider_item_definition))))
        (:then_8
          (return (LocalVariables|GetReturn%20Value)))))))

(fn IsHeroorPreviewsonHero
  (params (Item%20Definition :type Item%20Definition%20Base%20Object%20Reference) (SubSlotIndex :type Integer))
  (return (or (Utilities|Array|ContainsItem (Utilities|Array|MakeArray "Hero" "AthenaCharacter") (Class|FortItemFunctionLibrary|GetItemType Item%20Definition)) (Class|FortItemFunctionLibrary|ShouldShowPreviewonCurrentHero Item%20Definition SubSlotIndex))))

(fn IsJunoBuilding
  (params (FortItemDefinition :type Item%20Definition%20Base%20Object%20Reference))
  (bind _returnvalue_primaryassettype (Utilities|Struct|BreakPrimaryAssetId (AssetManager|GetPrimaryAssetIdfromObject FortItemDefinition)))
  (return (Utilities|Array|ContainsItem (Utilities|Array|MakeArray "JunoBuildingSet" "JunoBuildingProp") _returnvalue_primaryassettype)))
